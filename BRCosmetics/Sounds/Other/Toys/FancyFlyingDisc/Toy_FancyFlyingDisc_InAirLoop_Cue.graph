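# Sound cue for the looping in-air sound of the Fancy Flying Disc toy, in the gadgets sound class.
# A looping wave player feeds a modulator that slightly randomizes pitch and volume, then an enveloper that ramps volume up from half to full.
# The cue overrides attenuation with logarithmic distance falloff, and overrides concurrency to cap simultaneous instances per owner.
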SoundCue /BRCosmetics/Sounds/Other/Toys/FancyFlyingDisc/Toy_FancyFlyingDisc_InAirLoop_Cue
root=Enveloper_0
nodes [
 {
  id=Enveloper_0
  children [Modulator_0]
  properties {pitchCurve {editorCurveData {keys [{time=0}]}}, volumeCurve {editorCurveData {keys [<at 0 0.5>,<at 1 1>]}}}
 }
 {id=Modulator_0, children [WavePlayer_0], properties {pitchMax=0.9, pitchMin=0.8, volumeMax=1.1, volumeMin=1}}
 {
  id=WavePlayer_0
  properties {bLooping=true, soundWaveAssetPtr=&/Athena_Toy_FancyFlyingDsc_InAirLoop_01}
 }
]
properties {
 attenuationSettings=/Game/Sounds/Attenuation/Fort_ToyHit_Att
 bOverrideAttenuation=true
 bOverrideConcurrency=true
 maxDistance=2000
 pitchMultiplier=1.2
 soundClassObject=&Fort_Gadgets
 volumeMultiplier=0.4
 attenuationOverrides {distanceAlgorithm=Logarithmic, falloffDistance=2000, stereoSpread=512, attenuationShapeExtents {x=0}}
 concurrencyOverrides {bLimitToOwner=true, maxCount=2, volumeScale=0.7}
}
