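# Sound cue for the throw sound of the Fancy Flying Disc toy cosmetic.
# A Random node picks one of three throw wave variants with equal weight.
# It plays with emote attenuation and the emote music sound class, at a slightly boosted volume.
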SoundCue /BRCosmetics/Sounds/Other/Toys/FancyFlyingDisc/Toy_FancyFlyingDisc_Throw_Cue
root=Random_1
nodes [
 {id=Random_1, children [WavePlayer_0,WavePlayer_2,WavePlayer_5], properties {weights [1,1,1]}}
 {id=WavePlayer_0, properties {soundWaveAssetPtr=&/Athena_Toy_FancyFlyingDsc_Throw_01}}
 {id=WavePlayer_2, properties {soundWaveAssetPtr=&/Athena_Toy_FancyFlyingDsc_Throw_02}}
 {id=WavePlayer_5, properties {soundWaveAssetPtr=&/Athena_Toy_FancyFlyingDsc_Throw_03}}
]
properties {
 attenuationSettings=&Fort_Emotes_Athena_Att
 maxDistance=1600
 soundClassObject=&Fort_Music_Emote
 volumeMultiplier=1.25
}
