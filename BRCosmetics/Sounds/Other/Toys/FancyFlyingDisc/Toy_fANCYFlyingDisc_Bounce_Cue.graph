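# Sound cue for the bounce of the Fancy Flying Disc toy, built from its landing sounds.
# A Random node picks one of three equally weighted Land waves and feeds a Modulator that varies pitch and volume per play.
# The cue plays at reduced volume through the shared ball-bounce attenuation and is routed to the emote music sound class.
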SoundCue /BRCosmetics/Sounds/Other/Toys/FancyFlyingDisc/Toy_fANCYFlyingDisc_Bounce_Cue
root=Modulator_0
nodes [
 {id=Modulator_0, children [Random_1], properties {pitchMin=0.85, volumeMax=1}}
 {id=Random_1, children [WavePlayer_0,WavePlayer_1,WavePlayer_5], properties {weights [1,1,1]}}
 {id=WavePlayer_0, properties {soundWaveAssetPtr=&/Athena_Toy_FancyFlyingDisc_Land_02}}
 {id=WavePlayer_1, properties {soundWaveAssetPtr=&/Athena_Toy_FancyFlyingDisc_Land_01}}
 {id=WavePlayer_5, properties {soundWaveAssetPtr=&/Athena_Toy_FancyFlyingDisc_Land_03}}
]
properties {
 attenuationSettings=/Game/Sounds/Attenuation/Fort_Ball_Bounce_Athena_Att
 maxDistance=7000
 soundClassObject=&Fort_Music_Emote
 volumeMultiplier=0.5
}
